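; Construction Blueprint for an ornate dining table set prop in the Helios environment that tints its secondary meshes.
; The construction script runs the parent's, then stores the tint color from the first building texture data entry in Tint Color.
; It applies that color to four instanced static mesh components and, in a nested loop repeated each iteration, three table middle meshes.
; Destruction removes the Box2 component; other events are empty.
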(event EventBeginPlay)

(event Collision|EventActorBeginOverlap
  (params (OtherActor :type Actor%20Object%20Reference)))

(event EventTick
  (params (DeltaSeconds :type Float%20%28single-precision%29)))

(event EventDestroyed
  (Components|DestroyComponent (Variables|Default|GetBox2)))

(fn ConstructionScript
  (|CallParentFunction)
  (bind _outtexturedataarray (Building|GetBuildingTextureData))
  (bind _output_get (Variables|Default|SetTint%20Color (Variables|/Script/FortniteGame.BuildingTextureData|GetTintColor (Utilities|Array|Get(acopy) _outtexturedataarray))))
  (for _array_element (Utilities|Array|MakeArray (Variables|Default|GetInstancedStaticMesh2) (Variables|Default|GetInstancedStaticMesh1) (Variables|Default|GetInstancedStaticMesh) (Variables|Default|GetInstancedStaticMesh3) 0 0)
    (Default|SetColoronSecondaryMesh (Math|Conversions|ToLinearColor(Color) _output_get) _array_element)
    (for _array_element_2 (Utilities|Array|MakeArray (Variables|Default|GetSM_Furniture_OrnateDiningTableMiddle_A) (Variables|Default|GetSM_Furniture_OrnateDiningTableMiddle_B) (Variables|Default|GetSM_Furniture_OrnateDiningTableMiddle_B1))
      (Default|SetColoronSecondaryMesh (Math|Conversions|ToLinearColor(Color) (Variables|Default|GetTint%20Color)) _array_element_2))))
